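# Lit material for the nanobot mesh particle effect on the OysterKnock pickaxe, previewed on the bot static mesh.
# It is flagged for Niagara mesh particles and uses the Default Lit shading model.
# The small graph holds a constant color vector, a Particle Color node and a nanobot normal map sample, with no links between them or to material outputs shown.
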Material /BRCosmetics/Weapons/FORT_Melee/Pickaxe_OysterKnock/FX/M_NanoBot_FX
graph {
 properties {
  bAutomaticallySetUsageInEditor=false
  bUsedWithNiagaraMeshParticles=true
  previewMesh=&/SM_OysterKnock_Bot
  usedShadingModels=Default Lit
 }
 nodes [
  Constant4Vector()
  ParticleColor()
  TextureSample() {samplerType=SAMPLERTYPE_Normal, texture=/BRCosmetics/Weapons/FORT_Melee/Pickaxe_OysterKnock/Textures/T_OysterKnock_NanoBot_N}
 ]
}
