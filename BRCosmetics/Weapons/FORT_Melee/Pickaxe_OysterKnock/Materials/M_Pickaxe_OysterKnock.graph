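# Masked character-parent material for the OysterKnock pickaxe, with a weak-point break that grows back over time.
# A custom primitive data timestamp, delay and speed drive a clamped, decaying break amount, with a debug override.
# A pre-skinned position cutoff and UV box masks define the region, alongside a hexagon pattern texture and threshold gradient.
# Dithered temporal AA is blended into opacity mask between the PreFX and PostFX parent functions.
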Material /BRCosmetics/Weapons/FORT_Melee/Pickaxe_OysterKnock/Materials/M_Pickaxe_OysterKnock
graph {
 properties {
  bUseMaterialAttributes=true
  bUsedWithMeshDeformer=true
  bUsedWithMorphTargets=true
  bUsedWithSkeletalMesh=true
  blendMode=BLEND_Masked
  pixelDepthOffsetMode=PDOM_Legacy
  shadingModel=MSM_FromMaterialExpression
  usedShadingModels=Default Lit | Subsurface | Clear Coat | Subsurface Profile | Cloth
 }
 nodes [
  n1 = VertexInterpolator(VS=LocalPosition() {localOrigin=InstancePreSkinning}.XYZ)
  n2 = VectorParameter() {group=WeakPoints, parameterName=BreakPoint, defaultValue=<rgba -31.7594 0 37.2134 0>}
  Divide(
   A=ComponentMask(None=Subtract(A=@n1.PS, B=@n2.RGB)) {r=true}
   B=ScalarParameter() {defaultValue=26, group=WeakPoints, parameterName=BreakMaxDistance, sliderMax=50, sliderMin=0.01}
  )
  n3 = MaterialFunctionCall() {materialFunction=&MF_CharacterParent_2023_PreFX}
  n4 = Constant() {r=1}
  n5 = MaterialFunctionCall(Alpha Threshold=NamedRerouteUsage()) {materialFunction=&DitherTemporalAA}
  n6 = NamedRerouteUsage()
  MaterialFunctionCall(
   Material=SetMaterialAttributes!(MaterialAttributes=@n3.Result, Opacity Mask=LinearInterpolate(A=@n4, B=@n5.Result, Alpha=@n6)) {attributeSetTypes [#1]}
   M=@n3.M
   Pre Skinned Local Position=NamedRerouteDeclaration(None=@n3.Pre skinned Local Position) {name=PreskinnedLocalPos, nodeColor=<rgba 0.125 1 0 1>}
   Distance Blend=@n3.M
   Fresnel=@n3.M
  ) {materialFunction=&MF_CharacterParent_2023_PostFX}
  Multiply(A=NamedRerouteUsage(), B=NamedRerouteUsage())
  Multiply()
  n7 = Add(
   A=ScalarParameter() {bUseCustomPrimitiveData=true, group=WeakPoints, parameterName=GrowTimestamp_CPD, primitiveDataIndex=2}
   B=ScalarParameter() {defaultValue=0.5, group=WeakPoints, parameterName=GrowDelay}
  )
  n8 = Max(A=Constant(), B=Subtract(A=Time(), B=@n7))
  n9 = ScalarParameter() {defaultValue=1, desc=In seconds., group=WeakPoints, parameterName=GrowSpeed}
  n10 = OneMinus(None=Saturate(None=Divide(A=@n8, B=@n9)))
  n11 = Saturate(None=@n7)
  n12 = Clamp(
   None=Max(A=ScalarParameter() {desc=Only use for debug purposes., group=WeakPoints, parameterName=GrowTimeOverride}, B=Multiply(A=@n10, B=@n11))
   Max=ScalarParameter() {defaultValue=0.85, group=WeakPoints, parameterName=BreakMaxAmount, sliderMax=1}
  )
  NamedRerouteDeclaration(None=SquareRoot(None=NamedRerouteDeclaration(None=@n12) {name=Time, nodeColor=<rgba 0.6875 0 1 1>})) {name=SquaredTime, nodeColor=<rgba 0 1 0.3672 1>}
  n13 = Subtract(
   A=VertexInterpolator(VS=LocalPosition() {localOrigin=InstancePreSkinning}.XYZ).PS
   B=VectorParameter() {group=WeakPoints, parameterName=BreakPointPosition, defaultValue=<rgba -31.7594 0 37.2134 0>}.RGB
  )
  n14 = OneMinus(None=Divide(
    A=ComponentMask(None=@n13) {r=true}
    B=ScalarParameter() {defaultValue=26, group=WeakPoints, parameterName=BreakPointMaskDistance, sliderMax=50, sliderMin=0.01}
   ))
  n15 = TextureCoordinate()
  n16 = MaterialFunctionCall!(A=Saturate(None=@n15), B=Constant2Vector() {r=1}, Bounds=Constant2Vector() {g=1.225, r=1}, Edge Falloff=Constant()) {materialFunction=/Engine/Functions/Engine_MaterialFunctions01/Gradient/BoxMask-2D}
  n17 = MaterialFunctionCall(A=@n15, B=Constant2Vector() {g=0.95, r=0.3075}, Bounds=Constant2Vector() {g=0.075, r=0.075}, Edge Falloff=Constant()) {materialFunction=&BoxMask-2D}
  NamedRerouteDeclaration(None=Multiply(A=Sign(None=@n14), B=Max(A=@n16.Result, B=@n17.Result))) {name=Mask, nodeColor=<rgba 0.9219 0 1 1>}
  n18 = NamedRerouteUsage()
  n19 = NamedRerouteUsage()
  n20 = Multiply(
   A=Add(A=Multiply(A=@n18) {constB=0.75}, B=Subtract(A=@n19))
   B=Divide(B=ScalarParameter() {defaultValue=0.05, group=WeakPoints, parameterName=BreakSize, sliderMax=1, sliderMin=0.01})
  )
  NamedRerouteDeclaration(None=OneMinus(None=Saturate(None=@n20))) {name=LinearGradient, nodeColor=<rgba 1 0.6797 0 1>}
  n21 = TextureCoordinate()
  n22 = Constant2Vector() {g=1.625, r=2.5}
  n23 = TextureSample(UVs=Add(A=Multiply(A=@n21, B=@n22), B=Constant2Vector())) {samplerType=SAMPLERTYPE_LinearGrayscale, texture=/BRCosmetics/Weapons/FORT_Melee/Pickaxe_OysterKnock/Textures/T_OysterKnock_Pattern}
  NamedRerouteDeclaration(None=Saturate(None=@n23.R)) {name=HexagonNoise, nodeColor=<rgba 0.0781 0 1 1>}
  OneMinus(None=Constant() {r=1})
  Power(Base=@n23.R)
 ]
}
