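# Physics asset for the OysterKnock axe pickaxe, a cosmetic melee weapon mesh.
# It holds a single body on the base_general bone, shaped as one capsule rotated and offset along the bone's Z axis.
# There are no constraints, so the mesh simulates or collides as one rigid piece.
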PhysicsAsset /BRCosmetics/Weapons/FORT_Melee/Pickaxe_OysterKnock/Meshes/OysterKnock_Axe_PhysicsAsset
bodies {base_general [{shapeType=Capsule, length=32.2131, radius=17.2095, center=<0 0 16.6244>, rotation=<rot 0 -90 -90>}]}
constraints []
